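; Viewmodel for the MetaSound Metadata Editor tool that manages drag-style or index-based reordering of MetaSound outputs in a sort-order list.
; It holds entry viewmodels, moves an item from old to new index through successive adjacent swaps, then broadcasts a field change for GetReorderedViewmodels.
; GetReorderedViewmodels restamps each entry's index from its array position.
; ApplyReordering writes those positions to SortOrderIndex on each underlying MetaSound output editor viewmodel.
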(fn ApplyReordering
  (bind _output_get (LocalVariables|SetOutput%20Viewmodels (CallFunction|GetReorderedViewmodels)))
  (for (_array_element _array_index) _output_get
    (bind _entrydata (Variables|/Script/TechAudioToolsContent.TechAudioToolsEntryViewModel|GetEntryData _array_element)
      (:then
        (bind _asmetasound_output_editor_viewmodel (Utilities|Casting|CastToMetaSoundOutputEditorViewModel _entrydata))
        (Variables|/Script/TechAudioToolsMetaSoundEditor.MetaSoundOutputEditorViewModel|SetSortOrderIndex _array_index _asmetasound_output_editor_viewmodel))
      (:else))))

(fn GetReorderedViewmodels
  (for (_array_element _array_index) (Variables|Default|GetViewmodels)
    (Variables|/Script/TechAudioToolsContent.TechAudioToolsEntryViewModel|SetEntryIndex _array_index _array_element)
    (Utilities|Array|Add (LocalVariables|GetIndexed%20Viewmodels) _array_element))
  (return (LocalVariables|GetIndexed%20Viewmodels)))

(fn ReorderListItems
  (params (Old%20Index :type Integer) (New%20Index :type Integer))
  (if (!= Old%20Index New%20Index)
    (Utilities|FlowControl|Sequence
      (:then_0
        (LocalVariables|SetIndex%20Delta (- (Variables|GetNew%20Index) (Variables|GetOld%20Index))))
      (:then_1
        (Utilities|FlowControl|Sequence
          (:then_0
            (LocalVariables|SetStep (Math|Integer|SelectInt 1 -1 (> (LocalVariables|GetIndex%20Delta) 0))))
          (:then_1
            (Utilities|FlowControl|Sequence
              (:then_0
                (LocalVariables|SetNumber%20of%20Steps (Math|Integer|Absolute(Integer) (LocalVariables|GetIndex%20Delta))))
              (:then_1
                (Utilities|FlowControl|Sequence
                  (:then_0
                    (LocalVariables|SetCurrent%20Index (Variables|GetOld%20Index)))
                  (:then_1
                    (Utilities|FlowControl|Sequence
                      (:then_0
                        (for _ (range 1 (+ (LocalVariables|GetNumber%20of%20Steps) 1))
                          (LocalVariables|SetNext%20Index (+ (LocalVariables|GetCurrent%20Index) (LocalVariables|GetStep)))
                          (Utilities|Array|SwapArrayElements (Variables|Default|GetViewmodels) (LocalVariables|GetCurrent%20Index) (LocalVariables|GetNext%20Index))
                          (LocalVariables|SetCurrent%20Index (LocalVariables|GetNext%20Index))))
                      (:then_1
                        (Utilities|FlowControl|Sequence
                          (:then_0
                            (FieldNotification|BroadcastFieldValueChanged self "(FieldName=\"Get Reordered Viewmodels\")")))))))))))))))

(fn SetEntryViewmodelsforReordering
  (params (Viewmodels :type Array%20of%20Tech%20Audio%20Tools%20Entry%20Viewmodel%20Object%20References))
  (Variables|Default|SetViewmodels Viewmodels))
